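# Parent material for the StairGummy character's inverted-hull outlines, using unlit, masked, two-sided shading to show only back faces.
# World position offset pushes the hull along vertex, baked vertex-color or normal-map normals, with optional camera-direction shift.
# Thickness combines distance blending, an optional mask and panning noise, plus a hardcoded pre-skinned sphere mask around the nose.
# Flat or textured color passes through time-of-day and LUT adjustments, before cloth-wind and character post-FX functions.
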Material /BRCosmetics/Characters/Player/Female/Medium/Bodies/F_MED_StairGummy/Materials/M_StairGummy_InvHull_PARENT
graph {
 expressions {
  MaterialExpressionAbs_0 {class=MaterialExpressionAbs, inputs [None {from=MaterialExpressionDotProduct_1, output=""}], properties {}}
  MaterialExpressionAdd_1 {
   class=MaterialExpressionAdd
   inputs [
    A {from=MaterialExpressionChannelMaskParameter_0, output=R}
    B {from=MaterialExpressionConstant_3}
   ]
  }
  MaterialExpressionAdd_5 {
   class=MaterialExpressionAdd
   inputs [
    A {from=MaterialExpressionLinearInterpolate_4}
    B {from=MaterialExpressionReroute_9}
   ]
  }
  MaterialExpressionAdd_6 {
   class=MaterialExpressionAdd
   inputs [
    A {from=MaterialExpressionNamedRerouteUsage_36}
    B {from=MaterialExpressionStaticSwitchParameter_13}
   ]
  }
  MaterialExpressionAdd_7 {
   class=MaterialExpressionAdd
   inputs [
    A {from=MaterialExpressionMultiply_15}
    B {from=MaterialExpressionScalarParameter_18}
   ]
  }
  MaterialExpressionCameraPositionWS_1 {class=MaterialExpressionCameraPositionWS, inputs [], properties {}}
  MaterialExpressionCameraVectorWS_1 {class=MaterialExpressionCameraVectorWS, inputs [], properties {}}
  MaterialExpressionCeil_1 {
   class=MaterialExpressionCeil
   inputs [None {from=MaterialExpressionNamedRerouteDeclaration_15, output=""}]
   properties {desc=cut outlines mask}
  }
  MaterialExpressionChannelMaskParameter_0 {
   class=MaterialExpressionChannelMaskParameter
   inputs [None {from=MaterialExpressionNamedRerouteDeclaration_28, output=""}]
   properties {group=01_InvertedHull_Outlines, maskChannel=Blue, parameterName=OutlineThickness_Channel, sortPriority=3, defaultValue=<rgba 0 0 1 0>}
  }
  MaterialExpressionComposite_0 {class=MaterialExpressionComposite, inputs [], properties {subgraphName=Construct Base Inputs}}
  MaterialExpressionConstant3Vector_0 {class=MaterialExpressionConstant3Vector, properties {constant=<rgba 0 6.5 159.5 0>}}
  MaterialExpressionConstant_1 {class=MaterialExpressionConstant}
  MaterialExpressionConstant_3 {class=MaterialExpressionConstant, properties {desc=one bit used for opacity cutout, r=-0.009}}
  MaterialExpressionConstant_4 {class=MaterialExpressionConstant, properties {r=1}}
  MaterialExpressionConstant_5 {class=MaterialExpressionConstant, properties {r=1}}
  MaterialExpressionDivide_1 {
   class=MaterialExpressionDivide
   inputs [
    A {from=MaterialExpressionVectorParameter_2}
    B {from=""}
   ]
   properties {constB=10}
  }
  MaterialExpressionDotProduct_1 {
   class=MaterialExpressionDotProduct
   inputs [
    A {from=MaterialExpressionCameraVectorWS_1, output=""}
    B {from=MaterialExpressionTransform_4, output=""}
   ]
   properties {}
  }
  MaterialExpressionFresnel_0 {
   class=MaterialExpressionFresnel
   inputs [
    ExponentIn {from="", output=""}
    BaseReflectFractionIn {from="", output=""}
    Normal {from="", output=""}
   ]
   properties {baseReflectFraction=0, exponent=1}
  }
  MaterialExpressionLinearInterpolate_2 {
   class=MaterialExpressionLinearInterpolate
   inputs [
    A {from=MaterialExpressionReroute_2}
    B {from=MaterialExpressionScalarParameter_11}
    Alpha {from=MaterialExpressionNamedRerouteUsage_18}
   ]
  }
  MaterialExpressionLinearInterpolate_3 {
   class=MaterialExpressionLinearInterpolate
   inputs [
    A {from=MaterialExpressionReroute_8}
    B {from=MaterialExpressionMin_1}
    Alpha {from=MaterialExpressionPower_3}
   ]
  }
  MaterialExpressionLinearInterpolate_4 {
   class=MaterialExpressionLinearInterpolate
   inputs [
    A {from=""}
    B {from=MaterialExpressionMultiply_13}
    Alpha {from=MaterialExpressionStaticSwitchParameter_28}
   ]
   properties {constB=0}
  }
  MaterialExpressionLinearInterpolate_5 {
   class=MaterialExpressionLinearInterpolate
   inputs [
    A {from=MaterialExpressionStaticSwitchParameter_17}
    B {from=MaterialExpressionScalarParameter_19}
    Alpha {from=MaterialExpressionNamedRerouteUsage_39}
   ]
  }
  MaterialExpressionLocalPosition_0 {class=MaterialExpressionLocalPosition, inputs [], properties {localOrigin=InstancePreSkinning}}
  MaterialExpressionMaterialFunctionCall_1 {
   class=MaterialExpressionMaterialFunctionCall
   inputs [
    Material {from=MaterialExpressionStaticSwitchParameter_21}
    FX {from=MaterialExpressionNamedRerouteUsage_38}
   ]
   properties {materialFunction=/Game/Packages/Fortress_SharedMaterials/Anime_Master_Material/MFC_ClothWind}
  }
  MaterialExpressionMaterialFunctionCall_10 {
   class=MaterialExpressionMaterialFunctionCall
   inputs [In {from=MaterialExpressionNamedRerouteUsage_23}]
   properties {materialFunction=/Game/Packages/Fortress_SharedMaterials/Anime_Master_Material/MF_TODCustomColors}
  }
  MaterialExpressionMaterialFunctionCall_11 {
   class=MaterialExpressionMaterialFunctionCall
   inputs [
    Fade Offset {from=MaterialExpressionScalarParameter_10}
    Fade Distance {from=MaterialExpressionScalarParameter_9}
   ]
   properties {materialFunction=/Game/Global/Materials/Character_Parent_2023/Anime_Parent_2023/MaterialFunctions/MFC_DistanceFade_FOV_compensation}
  }
  MaterialExpressionMaterialFunctionCall_12 {
   class=MaterialExpressionMaterialFunctionCall
   inputs [
    Normal {from=MaterialExpressionTextureSampleParameter2D_7, output=RGB}
    Flatness {from=MaterialExpressionScalarParameter_12}
   ]
   properties {materialFunction=&FlattenNormal}
  }
  MaterialExpressionMaterialFunctionCall_14 {
   class=MaterialExpressionMaterialFunctionCall
   inputs [Material {from=""}]
   properties {materialFunction=/Game/Packages/Fortress_SharedMaterials/Base_Material_Functions/CharacterBrightnessandScale_Balance}
  }
  MaterialExpressionMaterialFunctionCall_15 {
   class=MaterialExpressionMaterialFunctionCall
   inputs []
   properties {materialFunction=/Game/Global/Materials/Character_Parent_2023/Anime_Parent_2023/MaterialFunctions/MFC_ActorAlighnedScreenPosition}
  }
  MaterialExpressionMaterialFunctionCall_16 {
   class=MaterialExpressionMaterialFunctionCall
   inputs [
    Material {from=MaterialExpressionMaterialFunctionCall_1, output=Result}
    M {from=MaterialExpressionReroute_8, output=M}
    Pre Skinned Local Position {from=MaterialExpressionReroute_9, output=Pre Skinned Local Position}
    Distance Blend {from=MaterialExpressionReroute_10, output=Distance Blend}
    Fresnel {from=MaterialExpressionReroute_11, output=Fresnel}
   ]
   properties {materialFunction=&MF_CharacterParent_2023_PostFX}
  }
  MaterialExpressionMaterialFunctionCall_7 {
   class=MaterialExpressionMaterialFunctionCall
   inputs [Color {from=MaterialExpressionMaterialFunctionCall_9, output=Result}]
   properties {materialFunction=/Game/Packages/Fortress_SharedMaterials/Anime_Master_Material/MF_TODColorAdjustment}
  }
  MaterialExpressionMaterialFunctionCall_8 {
   class=MaterialExpressionMaterialFunctionCall
   inputs [WPO {from=MaterialExpressionReroute_6}]
   properties {materialFunction=/Game/Packages/Fortress_SharedMaterials/Base_Material_Functions/Fortnite_HideElementsWithVertexColor_2022}
  }
  MaterialExpressionMaterialFunctionCall_9 {
   class=MaterialExpressionMaterialFunctionCall
   inputs [
    In {from=MaterialExpressionMaterialFunctionCall_10, output=Result}
    LUT {from=""}
    LUT Multiplier {from=""}
   ]
   properties {materialFunction=/Game/Packages/Fortress_SharedMaterials/Anime_Master_Material/MF_ApplyLUT}
  }
  MaterialExpressionMin_1 {
   class=MaterialExpressionMin
   inputs [
    A {from=MaterialExpressionReroute_8, output=""}
    B {from=MaterialExpressionNamedRerouteUsage_29, output=""}
   ]
   properties {}
  }
  MaterialExpressionMultiply_11 {
   class=MaterialExpressionMultiply
   inputs [
    A {from=MaterialExpressionStaticSwitchParameter_22}
    B {from=MaterialExpressionNamedRerouteUsage_26}
   ]
  }
  MaterialExpressionMultiply_12 {
   class=MaterialExpressionMultiply
   inputs [
    A {from=MaterialExpressionVertexColor_2}
    B {from=""}
   ]
   properties {constB=2}
  }
  MaterialExpressionMultiply_13 {
   class=MaterialExpressionMultiply
   inputs [
    A {from=MaterialExpressionNormalize_5}
    B {from=MaterialExpressionScalarParameter_14}
   ]
  }
  MaterialExpressionMultiply_14 {
   class=MaterialExpressionMultiply
   inputs [
    A {from=MaterialExpressionNamedRerouteUsage_35}
    B {from=MaterialExpressionScalarParameter_16}
   ]
   properties {constB=0.1}
  }
  MaterialExpressionMultiply_15 {
   class=MaterialExpressionMultiply
   inputs [
    A {from=MaterialExpressionSubtract_5}
    B {from=MaterialExpressionScalarParameter_17}
   ]
  }
  MaterialExpressionMultiply_18 {
   class=MaterialExpressionMultiply
   inputs [
    A {from=MaterialExpressionPower_4}
    B {from=""}
   ]
   properties {constB=10}
  }
  MaterialExpressionMultiply_5 {
   class=MaterialExpressionMultiply
   inputs [
    A {from=MaterialExpressionLinearInterpolate_5}
    B {from=MaterialExpressionNamedRerouteUsage_25}
   ]
  }
  MaterialExpressionMultiply_7 {
   class=MaterialExpressionMultiply
   inputs [
    A {from=MaterialExpressionTextureSampleParameter2D_2, output=RGB}
    B {from=MaterialExpressionScalarParameter_7}
   ]
  }
  MaterialExpressionMultiply_8 {
   class=MaterialExpressionMultiply
   inputs [
    A {from=MaterialExpressionCeil_1}
    B {from=MaterialExpressionOneMinus_0}
   ]
  }
  MaterialExpressionMultiply_9 {
   class=MaterialExpressionMultiply
   inputs [
    A {from=MaterialExpressionNamedRerouteUsage_19}
    B {from=MaterialExpressionStaticSwitchParameter_27}
   ]
  }
  MaterialExpressionNamedRerouteDeclaration_12 {
   class=MaterialExpressionNamedRerouteDeclaration
   inputs [None {from=MaterialExpressionMaterialFunctionCall_11, output=Result}]
   properties {name=DistanceFading, nodeColor=<rgba 1 0.3047 0 1>}
  }
  MaterialExpressionNamedRerouteDeclaration_14 {
   class=MaterialExpressionNamedRerouteDeclaration
   inputs [None {from=MaterialExpressionStaticSwitchParameter_25}]
   properties {name=WPO, nodeColor=<rgba 1 0.3281 0 1>}
  }
  MaterialExpressionNamedRerouteDeclaration_15 {
   class=MaterialExpressionNamedRerouteDeclaration
   inputs [None {from=MaterialExpressionAdd_1}]
   properties {name=ThicknessMask, nodeColor=<rgba 1 0.1689 0 1>}
  }
  MaterialExpressionNamedRerouteDeclaration_16 {
   class=MaterialExpressionNamedRerouteDeclaration
   inputs [None {from=MaterialExpressionMultiply_11}]
   properties {name=OpacityMask, nodeColor=<rgba 0 0.8125 1 1>}
  }
  MaterialExpressionNamedRerouteDeclaration_17 {
   class=MaterialExpressionNamedRerouteDeclaration
   inputs [None {from=MaterialExpressionStaticSwitchParameter_10}]
   properties {name=ColorRaw, nodeColor=<rgba 0 0.5781 1 1>}
  }
  MaterialExpressionNamedRerouteDeclaration_18 {
   class=MaterialExpressionNamedRerouteDeclaration
   inputs [None {from=MaterialExpressionMaterialFunctionCall_7, output=Result}]
   properties {name=ColorFinal, nodeColor=<rgba 1 0.6797 0 1>}
  }
  MaterialExpressionNamedRerouteDeclaration_19 {
   class=MaterialExpressionNamedRerouteDeclaration
   inputs [None {from=MaterialExpressionStaticSwitchParameter_19}]
   properties {name=Normal, nodeColor=<rgba 0.224 0.224 1 1>}
  }
  MaterialExpressionNamedRerouteDeclaration_21 {
   class=MaterialExpressionNamedRerouteDeclaration
   inputs [None {from=MaterialExpressionStaticSwitchParameter_20}]
   properties {name=Cinematics OutlineThicknessTextureOverride, nodeColor=<rgba 1 0 0.5391 1>}
  }
  MaterialExpressionNamedRerouteDeclaration_22 {
   class=MaterialExpressionNamedRerouteDeclaration
   inputs [None {from=MaterialExpressionPower_2}]
   properties {name=BakedNormalFresnel, nodeColor=<rgba 0 0.8144 1 1>}
  }
  MaterialExpressionNamedRerouteDeclaration_23 {
   class=MaterialExpressionNamedRerouteDeclaration
   inputs [None {from=MaterialExpressionNormalize_4}]
   properties {name=BakedNormals, nodeColor=<rgba 0 0.2789 1 1>}
  }
  MaterialExpressionNamedRerouteDeclaration_24 {
   class=MaterialExpressionNamedRerouteDeclaration
   inputs [None {from=MaterialExpressionStaticSwitchParameter_26}]
   properties {name=UVs, nodeColor=<rgba 1 0.6692 0 1>}
  }
  MaterialExpressionNamedRerouteDeclaration_25 {
   class=MaterialExpressionNamedRerouteDeclaration
   inputs [None {from=MaterialExpressionMaterialFunctionCall_15, output=UVs}]
   properties {name=ObjectAlighedScreenPosition, nodeColor=<rgba 1 0.867 0 1>}
  }
  MaterialExpressionNamedRerouteDeclaration_26 {
   class=MaterialExpressionNamedRerouteDeclaration
   inputs [None {from=MaterialExpressionAdd_7}]
   properties {name=ThicknessNoise, nodeColor=<rgba 1 0.2972 0 1>}
  }
  MaterialExpressionNamedRerouteDeclaration_27 {
   class=MaterialExpressionNamedRerouteDeclaration
   inputs [None {from=MaterialExpressionVertexColor_3, output=A}]
   properties {name=VertexColorAlpha, nodeColor=<rgba 0.8984 1 0 1>}
  }
  MaterialExpressionNamedRerouteDeclaration_28 {
   class=MaterialExpressionNamedRerouteDeclaration
   inputs [None {from=MaterialExpressionTextureSampleParameter2D_6, output=RGBA}]
   properties {name=FXmap, nodeColor=<rgba 0 0.4375 1 1>}
  }
  MaterialExpressionNamedRerouteDeclaration_29 {
   class=MaterialExpressionNamedRerouteDeclaration
   inputs [None {from=MaterialExpressionSaturate_0}]
   properties {name=NoseSideMask, nodeColor=<rgba 0 1 0.4375 1>}
  }
  MaterialExpressionNamedRerouteUsage_18 {class=MaterialExpressionNamedRerouteUsage}
  MaterialExpressionNamedRerouteUsage_19 {class=MaterialExpressionNamedRerouteUsage}
  MaterialExpressionNamedRerouteUsage_21 {class=MaterialExpressionNamedRerouteUsage}
  MaterialExpressionNamedRerouteUsage_22 {class=MaterialExpressionNamedRerouteUsage}
  MaterialExpressionNamedRerouteUsage_23 {class=MaterialExpressionNamedRerouteUsage}
  MaterialExpressionNamedRerouteUsage_24 {class=MaterialExpressionNamedRerouteUsage}
  MaterialExpressionNamedRerouteUsage_25 {class=MaterialExpressionNamedRerouteUsage}
  MaterialExpressionNamedRerouteUsage_26 {class=MaterialExpressionNamedRerouteUsage}
  MaterialExpressionNamedRerouteUsage_27 {class=MaterialExpressionNamedRerouteUsage}
  MaterialExpressionNamedRerouteUsage_28 {class=MaterialExpressionNamedRerouteUsage}
  MaterialExpressionNamedRerouteUsage_29 {class=MaterialExpressionNamedRerouteUsage}
  MaterialExpressionNamedRerouteUsage_30 {class=MaterialExpressionNamedRerouteUsage}
  MaterialExpressionNamedRerouteUsage_31 {class=MaterialExpressionNamedRerouteUsage}
  MaterialExpressionNamedRerouteUsage_32 {class=MaterialExpressionNamedRerouteUsage}
  MaterialExpressionNamedRerouteUsage_33 {class=MaterialExpressionNamedRerouteUsage}
  MaterialExpressionNamedRerouteUsage_34 {class=MaterialExpressionNamedRerouteUsage}
  MaterialExpressionNamedRerouteUsage_35 {class=MaterialExpressionNamedRerouteUsage}
  MaterialExpressionNamedRerouteUsage_36 {class=MaterialExpressionNamedRerouteUsage}
  MaterialExpressionNamedRerouteUsage_37 {class=MaterialExpressionNamedRerouteUsage}
  MaterialExpressionNamedRerouteUsage_38 {class=MaterialExpressionNamedRerouteUsage}
  MaterialExpressionNamedRerouteUsage_39 {class=MaterialExpressionNamedRerouteUsage}
  MaterialExpressionNormalize_3 {class=MaterialExpressionNormalize, inputs [VectorInput {from=MaterialExpressionMaterialFunctionCall_12, output=Result}]}
  MaterialExpressionNormalize_4 {class=MaterialExpressionNormalize, inputs [VectorInput {from=MaterialExpressionSubtract_3, output=""}]}
  MaterialExpressionNormalize_5 {class=MaterialExpressionNormalize, inputs [VectorInput {from=MaterialExpressionSubtract_4, output=""}]}
  MaterialExpressionOneMinus_0 {class=MaterialExpressionOneMinus, inputs [None {from=MaterialExpressionTwoSidedSign_1}]}
  MaterialExpressionOneMinus_4 {class=MaterialExpressionOneMinus, inputs [None {from=MaterialExpressionTextureSampleParameter2D_8, output=R}]}
  MaterialExpressionOneMinus_5 {class=MaterialExpressionOneMinus, inputs [None {from=MaterialExpressionAbs_0}]}
  MaterialExpressionPanner_1 {
   class=MaterialExpressionPanner
   inputs [
    Coordinate {from=MaterialExpressionMultiply_14}
    Time {from=""}
    Speed {from=MaterialExpressionDivide_1}
   ]
  }
  MaterialExpressionPinBase_0 {class=MaterialExpressionPinBase, inputs [], properties {}}
  MaterialExpressionPinBase_1 {
   class=MaterialExpressionPinBase
   inputs [
    M {from=MaterialExpressionConstant_1, output=""}
    Pre Skinned Local Position {from=MaterialExpressionShaderStageSwitch_0, output=""}
    Distance Blend {from=MaterialExpressionMaterialFunctionCall_14, output=Distance Blend}
    Fresnel {from=MaterialExpressionFresnel_0, output=""}
   ]
   properties {reroutePins [
     M {expression=&self:MaterialGraph_1.MaterialGraphNode_Composite_0.Construct Base Inputs.MaterialExpressionReroute_8}
     Pre Skinned Local Position {expression=&self:MaterialGraph_1.MaterialGraphNode_Composite_0.Construct Base Inputs.MaterialExpressionReroute_9}
     Distance Blend {expression=&self:MaterialGraph_1.MaterialGraphNode_Composite_0.Construct Base Inputs.MaterialExpressionReroute_10}
     Fresnel {expression=&self:MaterialGraph_1.MaterialGraphNode_Composite_0.Construct Base Inputs.MaterialExpressionReroute_11}
    ]}
  }
  MaterialExpressionPower_2 {
   class=MaterialExpressionPower
   inputs [
    Base {from=MaterialExpressionOneMinus_5}
    Exp {from=MaterialExpressionScalarParameter_13}
   ]
  }
  MaterialExpressionPower_3 {
   class=MaterialExpressionPower
   inputs [
    Base {from=MaterialExpressionNamedRerouteDeclaration_27}
    Exp {from=MaterialExpressionScalarParameter_15}
   ]
  }
  MaterialExpressionPower_4 {
   class=MaterialExpressionPower
   inputs [
    Base {from=MaterialExpressionShaderStageSwitch_1}
    Exp {from=""}
   ]
  }
  MaterialExpressionPreSkinnedPosition_0 {class=MaterialExpressionPreSkinnedPosition, inputs [], properties {}}
  MaterialExpressionReroute_2 {class=MaterialExpressionReroute, inputs [None {from=MaterialExpressionScalarParameter_8}]}
  MaterialExpressionReroute_6 {class=MaterialExpressionReroute, inputs [None {from=MaterialExpressionMultiply_5}]}
  MaterialExpressionReroute_7 {class=MaterialExpressionReroute, inputs [None {from=MaterialExpressionOneMinus_0}]}
  MaterialExpressionReroute_8 {class=MaterialExpressionReroute, inputs [None {from=MaterialExpressionStaticSwitchParameter_16}]}
  MaterialExpressionReroute_9 {class=MaterialExpressionReroute, inputs [None {from=MaterialExpressionStaticSwitchParameter_18}]}
  MaterialExpressionSaturate_0 {class=MaterialExpressionSaturate, inputs [None {from=MaterialExpressionMultiply_18}]}
  MaterialExpressionScalarParameter_10 {
   class=MaterialExpressionScalarParameter
   properties {defaultValue=50, group=01_InvertedHull_Outlines, parameterName=Fade_Offset, sliderMax=100, sortPriority=5}
  }
  MaterialExpressionScalarParameter_11 {
   class=MaterialExpressionScalarParameter
   properties {defaultValue=0.6, group=01_InvertedHull_Outlines, parameterName=LineThickness_Far, sliderMax=2, sortPriority=3}
  }
  MaterialExpressionScalarParameter_12 {
   class=MaterialExpressionScalarParameter
   properties {
    defaultValue=-0.1
    desc="values above 1.0 and below 0.0 can help with seperated geometry.\r\n-0.1 seems like a good wellyou to alight split geo."
    group=01_InvertedHull_Outlines
    parameterName=NormalMap_Flatness
    sliderMax=2
    sliderMin=-1
    sortPriority=5
   }
  }
  MaterialExpressionScalarParameter_13 {
   class=MaterialExpressionScalarParameter
   properties {defaultValue=5, group=01_InvertedHull_Outlines, parameterName=BakedNormalOpacity_Exponent, sliderMax=12}
  }
  MaterialExpressionScalarParameter_14 {
   class=MaterialExpressionScalarParameter
   properties {defaultValue=0.3, group=01_InvertedHull_Outlines, parameterName=CameraDirectionOffset, sliderMax=2, sortPriority=3}
  }
  MaterialExpressionScalarParameter_15 {
   class=MaterialExpressionScalarParameter
   properties {
    defaultValue=1
    desc="Controls the exponent curve of the alpha channel to mask out the Baked NormalOpacity Fresnel effect with the Vertex Color alpha channel. \r\nto improve the masking if the alpha channel is used for opacity masking, because of smooth normal for the hair."
    group=01_InvertedHull_Outlines
    parameterName=BakedNormalOpacity_AlphaExponent
    sliderMax=8
   }
  }
  MaterialExpressionScalarParameter_16 {
   class=MaterialExpressionScalarParameter
   properties {defaultValue=0.2, group=01_InvertedHull_Outlines, parameterName=ThicknessVariation_Tiling}
  }
  MaterialExpressionScalarParameter_17 {
   class=MaterialExpressionScalarParameter
   properties {defaultValue=1, group=01_InvertedHull_Outlines, parameterName=ThicknessVariation_Intensity, sliderMax=2}
  }
  MaterialExpressionScalarParameter_18 {
   class=MaterialExpressionScalarParameter
   properties {group=01_InvertedHull_Outlines, parameterName=ThicknessVariation_Bias, sliderMax=1, sliderMin=-1}
  }
  MaterialExpressionScalarParameter_19 {
   class=MaterialExpressionScalarParameter
   properties {group=01_InvertedHull_Outlines, parameterName=NoiseSideLineThickness, sliderMax=0.5, sortPriority=3}
  }
  MaterialExpressionScalarParameter_7 {
   class=MaterialExpressionScalarParameter
   properties {
    defaultValue=1
    desc=Multiplier for the Outline Color Texture
    group=01_InvertedHull_Outlines
    parameterName=Color_Intensity
    sliderMax=1
    sortPriority=3
   }
  }
  MaterialExpressionScalarParameter_8 {
   class=MaterialExpressionScalarParameter
   properties {defaultValue=0.1, desc="LineThickness (Near)", group=01_InvertedHull_Outlines, parameterName=LineThickness, sliderMax=0.5, sortPriority=3}
  }
  MaterialExpressionScalarParameter_9 {
   class=MaterialExpressionScalarParameter
   properties {defaultValue=1000, group=01_InvertedHull_Outlines, parameterName=Fade_Distance, sliderMax=1500, sliderMin=500, sortPriority=5}
  }
  MaterialExpressionSetMaterialAttributes_2 {
   class=MaterialExpressionSetMaterialAttributes
   inputs [
    MaterialAttributes {from="", output=""}
    Emissive Color {from=MaterialExpressionNamedRerouteUsage_24, output=""}
    Opacity Mask {from=MaterialExpressionNamedRerouteUsage_21, output=""}
    World Position Offset {from=MaterialExpressionNamedRerouteUsage_22, output=""}
   ]
   properties {attributeSetTypes [#1, #2, #3]}
  }
  MaterialExpressionSetMaterialAttributes_3 {
   class=MaterialExpressionSetMaterialAttributes
   inputs [
    MaterialAttributes {from="", output=""}
    Emissive Color {from=MaterialExpressionNamedRerouteUsage_27, output=""}
    Opacity Mask {from="", output=""}
    World Position Offset {from=MaterialExpressionNamedRerouteUsage_28, output=""}
   ]
   properties {
    desc=Debug
    attributeSetTypes [#1, #2, #3]
   }
  }
  MaterialExpressionShaderStageSwitch_0 {
   class=MaterialExpressionShaderStageSwitch
   inputs [
    PixelShader {from=MaterialExpressionVertexInterpolator_0, output=PS}
    VertexShader {from=MaterialExpressionPreSkinnedPosition_0, output=""}
   ]
   properties {}
  }
  MaterialExpressionShaderStageSwitch_1 {
   class=MaterialExpressionShaderStageSwitch
   inputs [
    PixelShader {from=MaterialExpressionVertexInterpolator_1, output=PS}
    VertexShader {from=MaterialExpressionSphereMask_0, output=""}
   ]
   properties {}
  }
  MaterialExpressionSphereMask_0 {
   class=MaterialExpressionSphereMask
   inputs [
    A {from=MaterialExpressionLocalPosition_0, output=XYZ}
    B {from=MaterialExpressionConstant3Vector_0, output=""}
    Radius {from="", output=""}
    Hardness {from="", output=""}
   ]
   properties {attenuationRadius=6, hardnessPercent=10}
  }
  MaterialExpressionStaticSwitchParameter_10 {
   class=MaterialExpressionStaticSwitchParameter
   inputs [
    True {from=MaterialExpressionMultiply_7, output=""}
    False {from=MaterialExpressionVectorParameter_1, output=RGB}
   ]
   properties {
    desc=Enable to use a texture to color the edge outlines of the mesh.
    group=01_InvertedHull_Outlines
    parameterName=UseDFL_ColorMap
    sortPriority=0
   }
  }
  MaterialExpressionStaticSwitchParameter_13 {
   class=MaterialExpressionStaticSwitchParameter
   inputs [
    True {from=MaterialExpressionLinearInterpolate_2, output=""}
    False {from=MaterialExpressionScalarParameter_8, output=""}
   ]
   properties {defaultValue=true, group=01_InvertedHull_Outlines, parameterName=UseDistanceFading, sortPriority=0}
  }
  MaterialExpressionStaticSwitchParameter_16 {
   class=MaterialExpressionStaticSwitchParameter
   inputs [
    True {from=MaterialExpressionMultiply_8, output=""}
    False {from=MaterialExpressionReroute_7, output=""}
   ]
   properties {group=01_InvertedHull_Outlines, parameterName=UseThicknessMap, sortPriority=0}
  }
  MaterialExpressionStaticSwitchParameter_17 {
   class=MaterialExpressionStaticSwitchParameter
   inputs [
    True {from=MaterialExpressionMultiply_9, output=""}
    False {from=MaterialExpressionStaticSwitchParameter_27, output=""}
   ]
   properties {group=01_InvertedHull_Outlines, parameterName=UseThicknessMap, sortPriority=0}
  }
  MaterialExpressionStaticSwitchParameter_18 {
   class=MaterialExpressionStaticSwitchParameter
   inputs [
    True {from=MaterialExpressionMaterialFunctionCall_8, output=Result}
    False {from=MaterialExpressionReroute_6, output=""}
   ]
   properties {group=03_Opacity Mask, parameterName=Use Vertex Colors for Mask}
  }
  MaterialExpressionStaticSwitchParameter_19 {
   class=MaterialExpressionStaticSwitchParameter
   inputs [
    True {from=MaterialExpressionTransform_3, output=""}
    False {from=MaterialExpressionStaticSwitchParameter_24, output=""}
   ]
   properties {desc=use normal map to prevent segmented geometry, group=01_InvertedHull_Outlines, parameterName=UseNormalMap, sortPriority=0}
  }
  MaterialExpressionStaticSwitchParameter_20 {
   class=MaterialExpressionStaticSwitchParameter
   inputs [
    True {from=MaterialExpressionOneMinus_4, output=""}
    False {from=MaterialExpressionConstant_4, output=""}
   ]
   properties {group=Cinematic, parameterName=UseCinematicsOutlineThicknessOverrideTexture}
  }
  MaterialExpressionStaticSwitchParameter_21 {
   class=MaterialExpressionStaticSwitchParameter
   inputs [
    True {from=MaterialExpressionSetMaterialAttributes_3, output=""}
    False {from=MaterialExpressionSetMaterialAttributes_2, output=""}
   ]
   properties {group=01_InvertedHull_Outlines, parameterName=DebugThicknessMap, sortPriority=90}
  }
  MaterialExpressionStaticSwitchParameter_22 {
   class=MaterialExpressionStaticSwitchParameter
   inputs [
    True {from=MaterialExpressionLinearInterpolate_3, output=""}
    False {from=MaterialExpressionStaticSwitchParameter_16, output=""}
   ]
   properties {
    desc=mask out this effect with the vertec color alpha channel
    group=01_InvertedHull_Outlines
    parameterName=UseBakedNormalOpacityFresnel
    sortPriority=0
   }
  }
  MaterialExpressionStaticSwitchParameter_24 {
   class=MaterialExpressionStaticSwitchParameter
   inputs [
    True {from=MaterialExpressionTransform_5, output=""}
    False {from=MaterialExpressionVertexNormalWS_1, output=""}
   ]
   properties {
    desc=use baked normals in the vertex colors instead of the surface normals.
    group=01_InvertedHull_Outlines
    parameterName=UseBakedNormals
    sortPriority=0
   }
  }
  MaterialExpressionStaticSwitchParameter_25 {
   class=MaterialExpressionStaticSwitchParameter
   inputs [
    True {from=MaterialExpressionAdd_5, output=""}
    False {from=MaterialExpressionReroute_9, output=""}
   ]
   properties {group=01_InvertedHull_Outlines, parameterName=OutlineOffsetInCameraDirection, sortPriority=0}
  }
  MaterialExpressionStaticSwitchParameter_26 {
   class=MaterialExpressionStaticSwitchParameter
   inputs [
    True {from=MaterialExpressionTextureCoordinate_3, output=""}
    False {from=MaterialExpressionTextureCoordinate_2, output=""}
   ]
   properties {
    desc="active to use the second UV channel (UV1) for the Textures"
    group=01_InvertedHull_Outlines
    parameterName=UseSecondUVChannel
    sortPriority=0
   }
  }
  MaterialExpressionStaticSwitchParameter_27 {
   class=MaterialExpressionStaticSwitchParameter
   inputs [
    True {from=MaterialExpressionAdd_6, output=""}
    False {from=MaterialExpressionStaticSwitchParameter_13, output=""}
   ]
   properties {group=01_InvertedHull_Outlines, parameterName=UseThicknessVariation, sortPriority=0}
  }
  MaterialExpressionStaticSwitchParameter_28 {
   class=MaterialExpressionStaticSwitchParameter
   inputs [
    True {from=MaterialExpressionNamedRerouteUsage_37, output=""}
    False {from=MaterialExpressionStaticSwitchParameter_29, output=""}
   ]
   properties {group=01_InvertedHull_Outlines, parameterName=UseOutlineOffsetVertexAlpha, sortPriority=0}
  }
  MaterialExpressionStaticSwitchParameter_29 {
   class=MaterialExpressionStaticSwitchParameter
   inputs [
    True {from=MaterialExpressionTextureSampleParameter2D_9, output=Mask}
    False {from=MaterialExpressionConstant_5, output=""}
   ]
   properties {group=01_InvertedHull_Outlines, parameterName=UseOutlineOffsetMap, sortPriority=0}
  }
  MaterialExpressionSubtract_3 {
   class=MaterialExpressionSubtract
   inputs [
    A {from=MaterialExpressionMultiply_12}
    B {from=""}
   ]
  }
  MaterialExpressionSubtract_4 {
   class=MaterialExpressionSubtract
   inputs [
    A {from=MaterialExpressionCameraPositionWS_1}
    B {from=MaterialExpressionWorldPosition_1, output=XYZ}
   ]
  }
  MaterialExpressionSubtract_5 {
   class=MaterialExpressionSubtract
   inputs [
    A {from=MaterialExpressionTextureSample_1, output=R}
    B {from=""}
   ]
   properties {constB=0.5}
  }
  MaterialExpressionTextureCoordinate_2 {class=MaterialExpressionTextureCoordinate}
  MaterialExpressionTextureCoordinate_3 {class=MaterialExpressionTextureCoordinate, properties {coordinateIndex=1}}
  MaterialExpressionTextureSampleParameter2D_2 {
   class=MaterialExpressionTextureSampleParameter2D
   inputs [
    UVs {from=MaterialExpressionNamedRerouteUsage_33, output=""}
    Apply View MipBias {from="", output=""}
   ]
   properties {
    desc="Outline Color Texture, use can use the ShadedMap."
    group=01_InvertedHull_Outlines
    parameterName=DFL_Color_Map
    sortPriority=1
    texture=&T_BlacksRGB
   }
  }
  MaterialExpressionTextureSampleParameter2D_6 {
   class=MaterialExpressionTextureSampleParameter2D
   inputs [
    UVs {from=MaterialExpressionNamedRerouteUsage_31, output=""}
    Apply View MipBias {from="", output=""}
   ]
   properties {
    group=01_InvertedHull_Outlines
    parameterName=STT_Map
    samplerType=SAMPLERTYPE_Masks
    sortPriority=1
    texture=/Game/Packages/Fortress_SharedMaterials/Anime_Master_Material/T_Outline_Mask
    channelNames {a="", b=Thickness, g=Threshold, r=Sheen}
   }
  }
  MaterialExpressionTextureSampleParameter2D_7 {
   class=MaterialExpressionTextureSampleParameter2D
   inputs [
    UVs {from=MaterialExpressionNamedRerouteUsage_32, output=""}
    Apply View MipBias {from="", output=""}
   ]
   properties {
    group=01_InvertedHull_Outlines
    parameterName=Normal_Map
    samplerType=SAMPLERTYPE_Normal
    sortPriority=1
    texture=&FlatNormalBlue
   }
  }
  MaterialExpressionTextureSampleParameter2D_8 {
   class=MaterialExpressionTextureSampleParameter2D
   inputs [
    UVs {from=MaterialExpressionNamedRerouteUsage_34, output=""}
    Apply View MipBias {from="", output=""}
   ]
   properties {
    group=Cinematic
    parameterName=CinematicsOutlineThicknessOverride
    samplerType=SAMPLERTYPE_Masks
    texture=/Game/Global/Textures/Default/Blanks/LargeEmpty
   }
  }
  MaterialExpressionTextureSampleParameter2D_9 {
   class=MaterialExpressionTextureSampleParameter2D
   inputs [
    UVs {from="", output=""}
    Apply View MipBias {from="", output=""}
   ]
   properties {
    group=01_InvertedHull_Outlines
    parameterName=OutlineOffset_Map
    sortPriority=1
    texture=&T_WhitesRGB
    channelNames {a="", b="", g="", r=Mask}
   }
  }
  MaterialExpressionTextureSample_1 {
   class=MaterialExpressionTextureSample
   inputs [
    UVs {from=MaterialExpressionPanner_1}
    Tex {from=""}
    Apply View MipBias {from=""}
   ]
   properties {samplerSource=SSM_Wrap_WorldGroupSettings, samplerType=SAMPLERTYPE_LinearColor, texture=/EpicBaseTextures/Noises/T_Perlin_Noise_D}
  }
  MaterialExpressionTransform_3 {class=MaterialExpressionTransform, inputs [None {from=MaterialExpressionNormalize_3}]}
  MaterialExpressionTransform_4 {
   class=MaterialExpressionTransform
   inputs [None {from=MaterialExpressionNamedRerouteUsage_30}]
   properties {transformSourceType=TRANSFORMSOURCE_Local}
  }
  MaterialExpressionTransform_5 {
   class=MaterialExpressionTransform
   inputs [None {from=MaterialExpressionNamedRerouteDeclaration_23}]
   properties {transformSourceType=TRANSFORMSOURCE_Local}
  }
  MaterialExpressionTwoSidedSign_1 {class=MaterialExpressionTwoSidedSign, inputs [], properties {}}
  MaterialExpressionVectorParameter_1 {
   class=MaterialExpressionVectorParameter
   properties {group=01_InvertedHull_Outlines, parameterName=Color, sortPriority=2, defaultValue=<rgba 0.005 0.005 0.005 1>}
  }
  MaterialExpressionVectorParameter_2 {
   class=MaterialExpressionVectorParameter
   properties {group=01_InvertedHull_Outlines, parameterName=ThicknessVariation_PannerSpeed, channelNames {a="", b="", g=Speed V, r=Speed U}}
  }
  MaterialExpressionVertexColor_2 {class=MaterialExpressionVertexColor, inputs [], properties {}}
  MaterialExpressionVertexColor_3 {class=MaterialExpressionVertexColor, inputs [], properties {}}
  MaterialExpressionVertexInterpolator_0 {class=MaterialExpressionVertexInterpolator, inputs [VS {from=MaterialExpressionPreSkinnedPosition_0, output=""}], properties {}}
  MaterialExpressionVertexInterpolator_1 {class=MaterialExpressionVertexInterpolator, inputs [VS {from=MaterialExpressionSphereMask_0, output=""}], properties {}}
  MaterialExpressionVertexNormalWS_1 {class=MaterialExpressionVertexNormalWS, inputs [], properties {}}
  MaterialExpressionWorldPosition_1 {class=MaterialExpressionWorldPosition, inputs [], properties {}}
 }
 properties {
  bAutomaticallySetUsageInEditor=false
  bEnableResponsiveAA=true
  bUseMaterialAttributes=true
  bUsedWithClothing=true
  bUsedWithMeshDeformer=true
  bUsedWithMorphTargets=true
  bUsedWithSkeletalMesh=true
  blendMode=BLEND_Masked
  pixelDepthOffsetMode=PDOM_Legacy
  previewMesh=/BRCosmetics/Characters/Player/Female/Medium/Heads/F_MED_StairGummy_Head/Meshes/F_MED_StairGummy_Head
  shadingModel=MSM_Unlit
  twoSided=true
  usedShadingModels=Unlit
 }
}
