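# Animation Blueprint for the F_MED_StairGummy female body cosmetic, adding secondary motion to its T-shirt.
# The pose copied from the parent mesh runs through a Control Rig, three chained spring AnimDynamics nodes on T-shirt bones, and a small fixed base-bone offset.
# Each update, a nativized parameter-asset function supplies the dynamics' gravity override.
# The Control Rig's IsCrouching input is partially raised during any crouch or sprint state, otherwise zero.
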AnimBlueprint /BRCosmetics/Characters/Player/Female/Medium/Bodies/F_MED_StairGummy/Meshes/F_MED_StairGummy_AnimBP
target_skeleton=&/F_MED_StairGummy_Skeleton
graphs [
 AnimGraph {nodes [
   n1 = Control Rig!(Source=Copy Pose From Mesh() {bUseAttachedParent=true}, Alpha, IsCrouching:double=Get IsCrouching() -> (IsCrouching:double)) -> (Pose) {controlRigAssetReference {blueprintRigClass=&/F_MED_StairGummy_CtrlRig.F_MED_StairGummy_CtrlRig_C}}
   n2 = Anim Dynamics - Bone: Tshirt_dyn_r!(ComponentPose=Local To Component(LocalPose=@n1), Alpha=1, GravityOverride:Vector=Get Gravity Override(), ExternalForce="0,0,0") -> (Pose) {
    angularDampingOverride=1
    angularSpringConstant=13000
    bAngularSpring=true
    bOverrideAngularDamping=true
    bUseGravityOverride=true
    gravityScale=0.35
    simulationSpace=RootRelative
    boundBone {boneName=Tshirt_dyn_r}
    componentAppliedLinearAccClamp=<3000 3000 3000>
    componentLinearAccScale=<0.4 0.4 0.5>
    componentLinearVelScale=<0.4 0.4 0.5>
    gravityOverride {z=-980}
    physicsBodyDefinitions [{
      boundBone {boneName=Tshirt_dyn_r}
      boxExtents=<20 25 25>
      constraintSetup {angularLimitsMax=<4 1 4>, angularLimitsMin=<-4 -4 -25>, angularTarget=<1 0 0>}
      localJointOffset=<4 0 0>
     }]
    simSpaceSettings {maxAngularAcceleration=100, maxAngularVelocity=10, simSpaceAngularAlpha=1}
   }
   n3 = Anim Dynamics - Bone: Tshirt_dyn_l!(ComponentPose=@n2, Alpha=1, GravityOverride:Vector=Get Gravity Override(), ExternalForce="0,0,0") -> (Pose) {
    angularDampingOverride=1
    angularSpringConstant=13000
    bAngularSpring=true
    bOverrideAngularDamping=true
    bUseGravityOverride=true
    gravityScale=0.35
    simulationSpace=RootRelative
    boundBone {boneName=Tshirt_dyn_l}
    componentAppliedLinearAccClamp=<3000 3000 3000>
    componentLinearAccScale=<0.4 0.4 0.5>
    componentLinearVelScale=<0.4 0.4 0.5>
    gravityOverride {z=-980}
    physicsBodyDefinitions [{
      boundBone {boneName=Tshirt_dyn_l}
      boxExtents=<20 25 25>
      constraintSetup {angularLimitsMax=<4 1 25>, angularLimitsMin=<-4 -4 -3>, angularTarget=<1 0 0>}
      localJointOffset=<4 0 0>
     }]
    simSpaceSettings {maxAngularAcceleration=100, maxAngularVelocity=10, simSpaceAngularAlpha=1}
   }
   n4 = Anim Dynamics - Bone: Tshirt_dyn_c!(ComponentPose=@n3, Alpha=1, GravityOverride:Vector=Get Gravity Override(), ExternalForce="0,0,0") -> (Pose) {
    angularDampingOverride=1
    angularSpringConstant=13000
    bAngularSpring=true
    bOverrideAngularDamping=true
    bUseGravityOverride=true
    gravityScale=0.35
    simulationSpace=RootRelative
    boundBone {boneName=Tshirt_dyn_c}
    componentAppliedLinearAccClamp=<3000 3000 3000>
    componentLinearAccScale=<0.4 0.4 0.5>
    componentLinearVelScale=<0.4 0.4 0.5>
    gravityOverride {z=-980}
    physicsBodyDefinitions [{
      boundBone {boneName=Tshirt_dyn_c}
      boxExtents=<25 30 30>
      constraintSetup {angularLimitsMax=<0 2 7>, angularLimitsMin=<0 -2 -20>, angularTarget=<1 0 0>}
      localJointOffset=<4 0 0>
     }]
    simSpaceSettings {maxAngularAcceleration=100, maxAngularVelocity=10, simSpaceAngularAlpha=1}
   }
   n5 = "Transform (Modify) Bone - Bone: Tshirt_base_c"(ComponentPose=@n4, Alpha=1, Translation="0,-1,0", Rotation="0,-2,0", Scale="1,1,1") {
    rotationMode=BMM_Replace
    rotationSpace=BCS_BoneSpace
    translationMode=BMM_Additive
    translationSpace=BCS_BoneSpace
    boneToModify {boneName=Tshirt_base_c}
    rotation {yaw=-2}
    translation {y=-1}
   }
   Output Pose(Result=Component To Local(ComponentPose=@n5)) {layerGroup=None}
  ]}
 EventGraph {nodes [
   Event BlueprintUpdateAnimation()
   > n1 = Sequence()
   > n2 = "Gravity Override State Function (Parameter Asset) (Nativized)"(
    CharacterPartAnimInstance=Self-Reference()
    GameTimeInSeconds=GetGameTimeInSeconds()
    DebugJointName=spine_05
    Parameters=&/F_MED_StairGummy_Parameters
   )
   > Set Gravity Override(Gravity Override=@n2)
   > Set RBAN Settings Out(RBAN Settings Out=@n2.OutRigidBodyAnimNodeParameters)
   n3 = Branch(
    execute=@n1.then_1
    Condition=OR Boolean!(A=Get bIsCrouching(), B=Get bIsCrouchSprinting(), C:bool=Get bIsSprinting(), D:bool=Get bIsTacticalSprinting()) -> (ReturnValue)
   )
   > Set IsCrouching(IsCrouching:double=0.7) -> (Output_Get:double)
   Set IsCrouching(execute=@n3.else, IsCrouching=0)
  ]}
]
variables [
 Gravity Override {type=Vector}
 Is Crouching {type=number}
 IsCrouching {type=number}
 RBAN Settings Out {type=FortRigidBodyAnimNodeParameters}
]
