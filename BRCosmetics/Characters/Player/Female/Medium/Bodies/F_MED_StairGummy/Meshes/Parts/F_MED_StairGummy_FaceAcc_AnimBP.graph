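# Animation blueprint for the face accessory mesh part of the StairGummy medium female player cosmetic body.
# The AnimGraph only copies the pose from the attached parent mesh via Copy Pose From Mesh and sends it straight to the output.
# The update event has no logic, and the declared facial-testing, gravity and rigid-body variables are not used in either graph.
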AnimBlueprint /BRCosmetics/Characters/Player/Female/Medium/Bodies/F_MED_StairGummy/Meshes/Parts/F_MED_StairGummy_FaceAcc_AnimBP
target_skeleton=&/F_MED_StairGummy_FaceAcc_Skeleton
graphs [
 AnimGraph {nodes [Output Pose(Result=Copy Pose From Mesh() {bUseAttachedParent=true}) {layerGroup=None}]}
 EventGraph {nodes [Event BlueprintUpdateAnimation()]}
]
variables [
 Facial_Testing {type=boolean}
 Gravity Override {type=Vector}
 Is_Facial_Testing {type=number}
 RBAN Settings Out {type=FortRigidBodyAnimNodeParameters}
]
